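# Material function for butterfly effects that produces a world position offset pushing sprite particle vertices toward the camera along the XY plane only.
# It zeroes Z on the particle-to-camera vector, normalizes it and scales it by the PushAmount input, which falls back to a scalar parameter of the same name.
# A second output exposes the raw, unflattened particle-to-camera vector.
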MaterialFunction /Game/Effects/Fort_Effects/Gameplay/Butterfly/MF_Butterfly_PushVertsToCamera_XY
graph {nodes [
  n1 = Subtract(A=CameraPositionWS(), B=ParticlePositionWS())
  n2 = Multiply(A=@n1, B=Constant3Vector() {constant=<rgba 1 1 0 0>})
  n3 = ScalarParameter() {defaultValue=23, parameterName=PushAmount}
  FunctionOutput(None=Multiply(
    A=Normalize(VectorInput=@n2)
    B=FunctionInput(Preview=@n3) {bUsePreviewValueAsDefault=true, inputName=PushAmount, inputType=FunctionInput_Scalar}
   )) {outputName=WPO Sprite Particle}
  FunctionOutput(None=@n1) {outputName=ParticleToCameraVec3, sortPriority=1}
 ]}
