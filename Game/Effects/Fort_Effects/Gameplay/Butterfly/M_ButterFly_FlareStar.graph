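# Unlit, alpha-composite particle sprite material that draws a cross-shaped star flare for the Butterfly gameplay effect.
# Particle alpha, a grayscale flare texture and a near-camera fade multiply into an opacity value for a depth fade.
# That faded mask scales particle color before the shared FX emissive-adjustment function.
# The graph also includes a Butterfly function that, by its name, pushes vertices toward the camera.
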Material /Game/Effects/Fort_Effects/Gameplay/Butterfly/M_ButterFly_FlareStar
graph {
 properties {
  bUsedWithParticleSprites=true
  blendMode=BLEND_AlphaComposite
  pixelDepthOffsetMode=PDOM_Legacy
  shadingModel=MSM_Unlit
  usedShadingModels=Unlit
 }
 nodes [
  n1 = ParticleColor()
  n2 = Multiply(
   A=Multiply(A=@n1.A, B=TextureSample() {samplerType=SAMPLERTYPE_Grayscale, texture=/Game/Effects/Fort_Effects/Textures/Flares/T_Cross_Flare}.R)
   B=MaterialFunctionCall(Distance=ScalarParameter() {defaultValue=200, parameterName=Distance}) {materialFunction=&Near_Camera_Fade}.Result
  )
  n3 = ScalarParameter() {defaultValue=5, parameterName=FadeDistance}
  MaterialFunctionCall(Emissive=Multiply(A=@n1.RGB, B=DepthFade(Opacity=@n2, FadeDistance=@n3))) {materialFunction=&MF_FX_EmissiveAdjustments}
  MaterialFunctionCall!(PushAmount) {materialFunction=&/MF_Butterfly_PushVertsToCamera_XY}
 ]
}
