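# Unlit alpha-composite particle sprite material for butterfly sparkle effects.
# The sparkle texture's green channel, cubed, blends between HDR outer and inner blue colors, masked by red and tinted by particle color through an emissive adjustment function.
# Opacity combines particle alpha, the red mask and a distance-based near-camera fade.
# A butterfly-specific function pushes vertices toward the camera.
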Material /Game/Effects/Fort_Effects/Gameplay/Butterfly/M_ButterFly_Sparkle
graph {
 properties {
  bUsedWithParticleSprites=true
  blendMode=BLEND_AlphaComposite
  pixelDepthOffsetMode=PDOM_Legacy
  shadingModel=MSM_Unlit
  usedShadingModels=Unlit
 }
 nodes [
  n1 = TextureSample() {texture=/Game/Effects/Fort_Effects/Textures/Sparkle/T_Sparkle_01}
  n2 = LinearInterpolate(
   A=VectorParameter() {parameterName=ColorOuter, defaultValue=<rgba 0 0.1615 5 0>}.RGB
   B=VectorParameter() {parameterName=ColorInner, defaultValue=<rgba 3.2812 30.2395 90 0>}.RGB
   Alpha=Power(Base=@n1.G) {constExponent=3}
  )
  n3 = ParticleColor()
  MaterialFunctionCall(Emissive=Multiply(A=Multiply(A=@n2, B=@n1.R), B=@n3.RGB)) {materialFunction=&MF_FX_EmissiveAdjustments}
  MaterialFunctionCall!(PushAmount) {materialFunction=&/MF_Butterfly_PushVertsToCamera_XY}
  n4 = ScalarParameter() {defaultValue=200, parameterName=Distance}
  Multiply(
   A=Multiply(A=@n3.A, B=MaterialFunctionCall(Distance=@n4) {materialFunction=&Near_Camera_Fade}.Result)
   B=@n1.R
  )
 ]
}
