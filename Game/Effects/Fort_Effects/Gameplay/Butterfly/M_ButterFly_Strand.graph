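# Unlit, two-sided alpha-composite mesh-particle material for glowing strands in the butterfly gameplay effect.
# Two panning smoke textures, one noise-distorted, build the masked strand, while a lifetime-driven sweep and bright tip push its color from low to high.
# Distance, depth and near-camera fades, optional Fresnel and Dynamic Parameter erosion shape opacity.
# World position offset adds curl noise, switchable spline thickening with camera-derived Z rotation, and an optional push toward camera.
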Material /Game/Effects/Fort_Effects/Gameplay/Butterfly/M_ButterFly_Strand
graph {
 properties {
  bAutomaticallySetUsageInEditor=false
  bUsedWithMeshParticles=true
  blendMode=BLEND_AlphaComposite
  pixelDepthOffsetMode=PDOM_Legacy
  shadingModel=MSM_Unlit
  twoSided=true
  usedShadingModels=Unlit
 }
 nodes [
  n1 = VectorParameter() {parameterName=ColorLow, defaultValue=<rgba 0.0303 0.0276 0.5208 0>}
  n2 = VectorParameter() {parameterName=ColorHigh, defaultValue=<rgba 0.1566 0.9429 2 0>}
  n3 = AppendVector(A=ScalarParameter() {defaultValue=1.5, parameterName=TileU_B}, B=Constant() {r=1})
  n4 = TextureCoordinate()
  n5 = MaterialFunctionCall(Float2=@n4) {materialFunction=&BreakOutFloat2Components}
  n6 = ScalarParameter() {defaultValue=0.5, parameterName=V Exponent}
  n7 = MaterialFunctionCall(XY=MaterialFunctionCall(X=@n4, Y=Power(Base=@n5.G, Exp=@n6)) {materialFunction=&MakeFloat2}.Result) {materialFunction=&Swizzle}
  n8 = DynamicParameter() {defaultValue=<rgba 1 0 1 1>, paramNames [Random,Erode,Param3,Param4]}
  n9 = Constant()
  n10 = ComponentMask(None=TextureSample(UVs=Panner() {bFractionalPart=true, speedY=0.1}) {samplerType=SAMPLERTYPE_LinearColor, texture=&T_NoisePack_LinearColor}.RGB) {g=true, r=true}
  n11 = ScalarParameter() {defaultValue=0.05, parameterName=UVDistortIntensity}
  n12 = Constant()
  n13 = Add(
   A=Add(A=@n7.YX, B=AppendVector(A=@n8.Random, B=@n9))
   B=Multiply(A=ConstantBiasScale(None=@n10) {bias=-0.5, scale=2}, B=AppendVector(A=@n11, B=@n12))
  )
  n14 = ScalarParameter() {defaultValue=0.46, parameterName=PanSpeedU_B}
  n15 = Time()
  n16 = TextureSample(UVs=Panner(Coordinate=Multiply(A=@n3, B=@n13), Time=Multiply(A=@n14, B=@n15)) {bFractionalPart=true, speedX=1}) {samplerType=SAMPLERTYPE_Grayscale, texture=&T_Smoke_Trail_Soft}
  n17 = ScalarParameter() {defaultValue=1.4, parameterName=Color Gain}
  n18 = ConstantBiasScale(None=@n16.R) {bias=-0.5, scale=0.2}
  n19 = ComponentMask(None=TextureCoordinate()) {g=true}
  n20 = ParticleRelativeTime()
  n21 = Multiply(
   A=@n16.R
   B=SphereMask(
    A=Add(A=@n18, B=@n19)
    B=OneMinus(None=@n20)
    Radius=ScalarParameter() {defaultValue=0.15, parameterName=SweepRadius}
    Hardness=ScalarParameter() {defaultValue=0.05, parameterName=SweepHardness}
   )
  )
  n22 = ScalarParameter() {defaultValue=4, parameterName=SweepScale}
  n23 = TextureCoordinate()
  n24 = Multiply(
   A=Power(
    Base=MaterialFunctionCall(Float2=@n23) {materialFunction=&BreakOutFloat2Components}.G
    Exp=ScalarParameter() {defaultValue=9, parameterName=TipExponent}
   ) {constExponent=8}
   B=ScalarParameter() {defaultValue=3, parameterName=TipScale}
  )
  n25 = Add(
   A=Power(Base=Multiply(A=@n16.R, B=@n17), Exp=ScalarParameter() {defaultValue=3, parameterName=Color Exponent})
   B=Add(A=Multiply(A=@n21, B=@n22), B=@n24)
  )
  n26 = ScalarParameter() {defaultValue=1, parameterName=Fresnel Brightness}
  n27 = ScalarParameter() {defaultValue=1.5, parameterName=Fresnel Exponent}
  n28 = Multiply(A=AppendVector(A=ScalarParameter() {defaultValue=0.9, parameterName=TileU_A}, B=Constant() {r=1}), B=@n7.YX)
  n29 = Multiply(A=ScalarParameter() {defaultValue=0.72, parameterName=PanSpeedU_A}, B=Time())
  n30 = Multiply(
   A=TextureSample(UVs=Panner(Coordinate=@n28, Time=@n29) {bFractionalPart=true, speedX=1}) {texture=&T_Inky_Smoke_Tile}.R
   B=ScalarParameter() {defaultValue=0.6, parameterName=BGScale}
  )
  n31 = SphereMask(
   A=MaterialFunctionCall(Float2=TextureCoordinate()) {materialFunction=&BreakOutFloat2Components}.R
   B=Constant() {r=0.5}
   Radius=ScalarParameter() {defaultValue=0.5, parameterName=VMaskRadius}
   Hardness=ScalarParameter() {defaultValue=0.6, parameterName=VMaskHardness}
  )
  n32 = Add(A=Multiply(A=Add(A=@n30, B=@n16.R), B=@n31), B=Multiply(A=ScalarParameter() {defaultValue=0.2, parameterName=BackOpacity}, B=@n31))
  n33 = Constant() {r=0.5}
  n34 = SphereMask(
   A=MaterialFunctionCall(Float2=TextureCoordinate()) {materialFunction=&BreakOutFloat2Components}.G
   B=@n33
   Radius=@n33
   Hardness=ScalarParameter() {defaultValue=0.25, parameterName=UMaskHardness}
  )
  n35 = MaterialFunctionCall(
   Opacity In=Add(A=Multiply(A=@n32, B=@n34), B=Multiply(A=@n31, B=@n24))
   Position=WorldPosition().XYZ
   Fade_Start_Distance=ScalarParameter() {defaultValue=100, parameterName=Fade_Start_Distance}
   Fade_End_Distance=ScalarParameter() {defaultValue=20, parameterName=Fade_End_Distance}
  ) {materialFunction=&MF_Distance_SyncedFade_FX}
  n36 = StaticSwitchParameter(
   True=Multiply(
    A=MaterialFunctionCall(Brightness=@n26, Exponent=@n27) {materialFunction=&MF_Fresnel_FX}.Fresnel
    B=@n35.Result
   )
   False=@n35.Result
  ) {parameterName=UseFresnel}
  n37 = ScalarParameter() {defaultValue=15, parameterName=FadeDistance}
  n38 = Multiply(
   A=Subtract(
    A=DepthFade(Opacity=@n36, FadeDistance=@n37)
    B=DynamicParameter() {defaultValue=<rgba 1 0 1 1>, paramNames [Random,Erode,Param3,Param4]}.Erode
   )
   B=MaterialFunctionCall(Distance=ScalarParameter() {defaultValue=200, parameterName=Distance}) {materialFunction=&Near_Camera_Fade}.Result
  )
  n39 = ParticleColor()
  MaterialFunctionCall(Emissive=Multiply(A=LinearInterpolate(A=@n1.RGB, B=@n2.RGB, Alpha=@n25), B=Multiply(A=@n38, B=@n39.A))) {materialFunction=&MF_FX_EmissiveAdjustments}
  n40 = MaterialFunctionCall(
   WidthBase=ScalarParameter() {defaultValue=10, parameterName=SplineWidthBase}
   WidthTip=ScalarParameter() {defaultValue=3, parameterName=SplineWidthTip}
  ) {materialFunction=&SplineThicken}
  n41 = Add(
   A=Subtract(A=WorldPosition().XYZ, B=ParticlePositionWS())
   B=DynamicParameter() {defaultValue=<rgba 1 0 1 1>, paramNames [Random,Erode,Param3,Param4]}.Random
  )
  n42 = Multiply(A=Time(), B=VectorParameter() {parameterName=CurlPan, defaultValue=<rgba 10 0 0 0>}.RGB)
  n43 = Divide(
   A=Transform(None=Add(A=@n41, B=@n42)) {transformSourceType=TRANSFORMSOURCE_World, transformType=TRANSFORM_Local}
   B=VectorParameter() {parameterName=CurlTileScale, defaultValue=<rgba 120 120 120 0>}.RGB
  )
  n44 = Multiply(
   A=@n34
   B=Multiply(
    A=VectorNoise(World Position=@n43) {noiseFunction=VNF_CurlALU}
    B=VectorParameter() {parameterName=CurlIntensity, defaultValue=<rgba 1.75 1.75 1.75 0>}.RGB
   )
  )
  n45 = Constant3Vector() {constant=<rgba 0 0 1 0>}
  n46 = MaterialFunctionCall(Float3=Normalize(VectorInput=MaterialFunctionCall!(PushAmount) {materialFunction=&/MF_Butterfly_PushVertsToCamera_XY}.ParticleToCameraVec3)) {materialFunction=&BreakOutFloat3Components}
  n47 = OneMinus(None=Divide(A=Arctangent2Fast(Y=@n46.R, X=@n46.R)) {constB=6.2832})
  n48 = ParticlePositionWS()
  n49 = WorldPosition()
  n50 = StaticSwitchParameter(
   True=Add(
    A=Add(A=@n40.WorldPositionOffset, B=@n44)
    B=RotateAboutAxis(NormalizedRotationAxis=@n45, RotationAngle=@n47, PivotPoint=@n48, Position=@n49.XYZ)
   )
   False=@n44
  ) {defaultValue=true, parameterName=SplineRotateWPO}
  StaticSwitchParameter(
   True=Add(
    A=@n50
    B=MaterialFunctionCall() {materialFunction=&MF_Butterfly_PushVertsToCamera_XY}.WPO Sprite Particle
   )
   False=@n50
  ) {parameterName=UsePushWPO}
 ]
}
