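# Unlit, two-sided translucent material for the butterfly effect's glowing character flipbook sprites.
# Vertex red selects each sprite's 4x4 SubUV color and mask frame, with a center flare added and color fading toward glow by distance.
# World position offset makes sprites, pushes them from the object origin and passes through MF_SpawnInBetween.
# A timestamp transition compared against inverted frame index switches opacity on progressively, highest frames first.
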Material /Game/Effects/Fort_Effects/Gameplay/Butterfly/M_Butterfly_CharacterFlipbook
graph {
 properties {
  bEnableResponsiveAA=true
  blendMode=BLEND_Translucent
  pixelDepthOffsetMode=PDOM_Legacy
  shadingModel=MSM_Unlit
  translucencyLightingMode=TLM_VolumetricDirectional
  twoSided=true
  usedShadingModels=Unlit
 }
 nodes [
  n1 = ScalarParameter() {defaultValue=16, parameterName=Frames}
  n2 = Ceil(None=Multiply(A=VertexColor().R, B=@n1))
  n3 = MaterialFunctionCall!(
   Texture=TextureObject() {texture=&/CharacterFlipBook_D}
   UVs
   SubImages=Constant() {r=4}
   Frame=@n2
  ) {materialFunction=/Engine/Functions/Engine_MaterialFunctions03/Texturing/SubUV_Function}
  n4 = VectorParameter() {parameterName=Glow, defaultValue=<rgba 0.9375 2.8159 4 1>}
  n5 = Multiply(A=ConstantBiasScale(None=@n3.RGB), B=Multiply(A=@n4.RGB, B=@n4.RGB))
  n6 = ScalarParameter() {defaultValue=0.1, parameterName=ColorLerp}
  n7 = WorldPosition()
  n8 = ScalarParameter() {defaultValue=800000, parameterName="Fade_Start_Distance (S)"}
  n9 = ScalarParameter() {defaultValue=10000, parameterName="Fade_End_Distance (S)"}
  n10 = TextureSample(UVs=MaterialFunctionCall(Texture Scale=ScalarParameter() {defaultValue=1.5, parameterName=GlowCenterScale}) {materialFunction=&ScaleUVsByCenter}.UVs) {texture=/Game/Effects/Fort_Effects/Textures/Flares/T_Flare_Spikeball}
  Add(
   A=LinearInterpolate(
    A=LinearInterpolate(A=@n3.RGB, B=@n5, Alpha=@n6)
    B=@n4.RGB
    Alpha=MaterialFunctionCall(Position=@n7.XYZ, Fade_Start_Distance=@n8, Fade_End_Distance=@n9) {materialFunction=&MF_Distance_SyncedFade_FX}.Result
   )
   B=Multiply(A=@n4.RGB, B=@n10.RGB)
  )
  n11 = WorldPosition()
  n12 = ScalarParameter() {defaultValue=800, parameterName=Size}
  n13 = Subtract(A=WorldPosition().XYZ, B=ObjectPositionWS())
  n14 = VectorParameter() {parameterName=PushScale, defaultValue=<rgba 255 255 35 0>}
  MaterialFunctionCall!(WPO=Add(
    A=MaterialFunctionCall!(Center Location=@n11.XYZ, 0-1 UVs, XY scale=@n12, OverridePivot, "Normals (Optional)") {materialFunction=&Sprite}.World position offset
    B=Multiply(A=@n13, B=@n14.RGB)
   )) {materialFunction=/Game/Athena/Environments/Nexus/Materials/MF_SpawnInBetween}
  n15 = Power(Base=@n10.RGB, Exp=ScalarParameter() {defaultValue=1, parameterName=FlareExponent})
  n16 = ScalarParameter() {defaultValue=3, parameterName=FlareOpacityScale}
  n17 = TextureObject() {samplerType=SAMPLERTYPE_Masks, texture=&/CharacterFlipBook_M}
  n18 = Constant() {r=4}
  n19 = Divide(A=@n2, B=@n1)
  n20 = ScalarParameter() {parameterName=TimestampA}
  n21 = ScalarParameter() {defaultValue=50, parameterName=TransitionLength_In}
  n22 = Constant() {r=1}
  Multiply(
   A=Add(
    A=Multiply(A=@n15, B=@n16)
    B=MaterialFunctionCall(Texture=@n17, SubImages=@n18, Frame=@n2) {materialFunction=&SubUV_Function}.RGB
   )
   B=If(
    A=OneMinus(None=@n19)
    B=MaterialFunctionCall(TimestampA=@n20, TransitionLength_In=@n21) {materialFunction=&MF_Timestamp_InOut_01}.Linear
    "A > B"=Constant()
    "A == B"=@n22
    "A < B"=@n22
   )
  )
 ]
}
