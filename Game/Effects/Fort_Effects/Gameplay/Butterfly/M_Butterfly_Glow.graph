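# Unlit additive sprite particle material that draws a soft round glow for the butterfly gameplay effect.
# A sphere mask on the UVs is sharpened by an exponent and tinted by particle color.
# Depth fade and near-camera fade soften it before it passes through the shared FX emissive adjustments function.
# It also includes a butterfly function named for pushing vertices toward the camera in XY.
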Material /Game/Effects/Fort_Effects/Gameplay/Butterfly/M_Butterfly_Glow
graph {
 properties {
  bEnableResponsiveAA=true
  bUsedWithParticleSprites=true
  blendMode=BLEND_Additive
  pixelDepthOffsetMode=PDOM_Legacy
  shadingModel=MSM_Unlit
  usedShadingModels=Unlit
 }
 nodes [
  MaterialFunctionCall!(PushAmount) {materialFunction=&/MF_Butterfly_PushVertsToCamera_XY}
  n1 = TextureCoordinate()
  n2 = Constant() {r=0.5}
  n3 = ScalarParameter() {defaultValue=0.7, parameterName=Radius}
  n4 = ScalarParameter() {defaultValue=0.2, parameterName=Hardness}
  n5 = Multiply(
   A=ParticleColor().RGB
   B=Power(Base=SphereMask(A=@n1, B=@n2, Radius=@n3, Hardness=@n4), Exp=ScalarParameter() {defaultValue=3, parameterName=Exponent})
  )
  n6 = ScalarParameter() {defaultValue=15, parameterName=FadeDistance}
  n7 = ScalarParameter() {defaultValue=200, parameterName=Distance}
  MaterialFunctionCall(Emissive=Multiply(
    A=DepthFade(Opacity=@n5, FadeDistance=@n6)
    B=MaterialFunctionCall(Distance=@n7) {materialFunction=&Near_Camera_Fade}.Result
   )) {materialFunction=&MF_FX_EmissiveAdjustments}
 ]
}
