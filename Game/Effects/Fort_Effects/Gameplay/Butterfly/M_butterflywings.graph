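# Translucent, unlit, two-sided butterfly wing material for Niagara particles, with procedural flapping.
# Vertex color R and G mask opposite sine-driven rotations about the mesh particle pivot, bent along U and tuned by Amplitude, Period, Curvature and Offset dynamic parameters.
# A separate sine chain computes a vertex-color-weighted local-Z offset.
# The texture, optionally desaturated, takes particle color, time-of-day tints and emissive adjustments, with opacity from both alphas.
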Material /Game/Effects/Fort_Effects/Gameplay/Butterfly/M_butterflywings
graph {
 properties {
  bUsedWithNiagaraMeshParticles=true
  bUsedWithNiagaraRibbons=true
  blendMode=BLEND_Translucent
  pixelDepthOffsetMode=PDOM_Legacy
  shadingModel=MSM_Unlit
  twoSided=true
  usedShadingModels=Unlit
 }
 nodes [
  n1 = Normalize(VectorInput=Transform(None=Constant3Vector() {constant=<rgba 0 1 0 0>}) {transformSourceType=TRANSFORMSOURCE_ParticleWorld})
  n2 = Constant() {r=0.08}
  n3 = Time()
  n4 = DynamicParameter() {paramNames [Amplitude,Period,Curvature,Offset]}
  n5 = Add(A=Sine(None=Multiply(A=@n3, B=@n4.Period)), B=@n4.Offset) {constB=-0.35}
  n6 = ComponentMask(None=TextureCoordinate()) {r=true}
  n7 = Multiply(A=Multiply(A=Multiply(A=@n2, B=@n5), B=Add(A=@n6, B=@n4.Curvature) {constB=0.5}), B=@n4.Amplitude)
  n8 = MaterialFunctionCall() {materialFunction=&ObjectPivotPoint}
  n9 = WorldPosition()
  n10 = Normalize(VectorInput=Transform(None=Constant3Vector() {constant=<rgba 0 1 0 0>}) {transformSourceType=TRANSFORMSOURCE_ParticleWorld})
  n11 = Constant() {r=-0.08}
  n12 = Time()
  n13 = Add(A=Sine(None=Multiply(A=@n12, B=@n4.Period) {constB=5}), B=@n4.Offset) {constB=-0.35}
  n14 = ComponentMask(None=TextureCoordinate()) {r=true}
  n15 = Multiply(A=@n4.Amplitude, B=Multiply(A=Multiply(A=@n11, B=@n13), B=Add(A=@n14, B=@n4.Curvature) {constB=0.5}))
  n16 = MaterialFunctionCall() {materialFunction=&ObjectPivotPoint}
  n17 = WorldPosition()
  Add(
   A=Multiply(
    A=VertexColor().R
    B=RotateAboutAxis(NormalizedRotationAxis=@n1, RotationAngle=@n7, PivotPoint=@n8.Mesh Particle Pivot Location, Position=@n9.XYZ)
   )
   B=Multiply(
    A=VertexColor().G
    B=RotateAboutAxis(NormalizedRotationAxis=@n10, RotationAngle=@n15, PivotPoint=@n16.Mesh Particle Pivot Location, Position=@n17.XYZ)
   )
  )
  n18 = Constant3Vector() {constant=<rgba 0.2 0.25 0.5 0>}
  n19 = ParticleColor()
  n20 = TextureSampleParameter2D() {parameterName=Texture, texture=&/T_butterfly_big}
  n21 = Multiply(A=@n19.RGB, B=Desaturation(None=@n20.RGB, Fraction=ScalarParameter() {parameterName=Desaturation}))
  n22 = Constant3Vector() {constant=<rgba 0.5 0.35 0.35 0>}
  n23 = Constant3Vector() {constant=<rgba 0.5 0.5 0.5 0>}
  MaterialFunctionCall(Emissive=MaterialFunctionCall(Color_Morning=Multiply(A=@n18, B=@n21), Color_Day=@n21, Color_Evening=Multiply(A=@n22, B=@n21), Color_Night=Multiply(A=@n21, B=@n23)) {materialFunction=&Color_Time_of_Day}.Result) {materialFunction=&MF_FX_EmissiveAdjustments}
  Multiply(A=@n19.A, B=@n20.A)
  n24 = DynamicParameter() {paramNames [Amplitude,Period,Curvature,Offset]}
  n25 = Time()
  n26 = Add(A=Sine(None=Multiply(A=@n24.Period, B=@n25)) {period=0.75}, B=@n24.Curvature)
  n27 = Transform(None=Constant3Vector() {constant=<rgba 0 0 1 0>}) {transformSourceType=TRANSFORMSOURCE_Local}
  Multiply(A=VertexColor(), B=Multiply(A=Multiply(A=@n27, B=@n26), B=@n24.Amplitude) {constB=5})
 ]
}
